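; Blueprint for the Mechanical Engineer owl pet cosmetic, switching particle effects as its owning Fortnite player pawn skydives or parachutes.
; BeginPlay activates the eye effect and binds handlers to the pawn's skydive and parachute begin/end events.
; Begin handlers swap eyes for a spark burst, then start wing air trails after a delay; end handlers restore eyes, deactivate only the left trail, then spark again.
; An unbound KillEvent fires a kill flash.
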(event Custom|Parachute
  ;[1 1 1] The skydive event,
  (Development|PrintString "START" true true "(R=1.000000,G=0.000000,B=0.138157,A=1.000000)")
  (Components|Activation|Deactivate (Variables|Default|GetP_Backpack_MED_Mechanical_Engineer_Eyes))
  (Components|Activation|Activate (Variables|Default|GetP_Backpack_MED_Mechanical_Engineer_SparkBurst) true)
  (Utilities|FlowControl|Delay 0.5)
  (Components|Activation|Activate (Variables|Default|GetFX_AirTrail_R) false)
  (Effects|Particles|Trails|BeginTrails (Variables|Default|GetFX_AirTrail_L) "FX_TrailL1" "FX_TrailL2" "ETrailWidthMode_FromCentre" 1.0)
  (Effects|Particles|Trails|BeginTrails (Variables|Default|GetFX_AirTrail_R) "FX_TrailR1" "FX_TrailR2" "ETrailWidthMode_FromCentre" 1.0))

(event Collision|EventActorBeginOverlap
  (params (OtherActor :type Actor%20Object%20Reference))
  (|CallParentFunction OtherActor) ; This node is disabled and will not be called.\nDrag off pins to build functionality.
)

(event EventTick
  (params (DeltaSeconds :type Float%20%28single-precision%29))
  (|CallParentFunction DeltaSeconds) ; This node is disabled and will not be called.\nDrag off pins to build functionality.
)

(event Custom|BeginSkydiving
  ;[1 1 1] The skydive event,
  (Development|PrintString "START" true true "(R=1.000000,G=0.000000,B=0.138157,A=1.000000)")
  (Components|Activation|Deactivate (Variables|Default|GetP_Backpack_MED_Mechanical_Engineer_Eyes))
  (Components|Activation|Activate (Variables|Default|GetP_Backpack_MED_Mechanical_Engineer_SparkBurst) true)
  (Utilities|FlowControl|Delay 0.5)
  (Components|Activation|Activate (Variables|Default|GetFX_AirTrail_R) false)
  (Effects|Particles|Trails|BeginTrails (Variables|Default|GetFX_AirTrail_L) "FX_TrailL1" "FX_TrailL2" "ETrailWidthMode_FromCentre" 1.0)
  (Effects|Particles|Trails|BeginTrails (Variables|Default|GetFX_AirTrail_R) "FX_TrailR1" "FX_TrailR2" "ETrailWidthMode_FromCentre" 1.0))

(event EventBeginPlay
  (bind _returnvalue (CallFunction|GetFortPlayerPawn))
  (|CallParentFunction)
  (Utilities|FlowControl|Sequence
    (:then_0
      (Class|FortAsyncActionGameStateReady|GameStateReadyAsync
        (:then
          _)
        (:OnReady
          _))
      (EventDispatchers|BindEventtoOnBeginParachuteMovementEvent _returnvalue (delegate Custom|Parachute))
      (EventDispatchers|BindEventtoOnBeginSkydivingEvent _returnvalue (delegate Custom|BeginSkydiving))
      (EventDispatchers|BindEventtoOnEndSkydivingEvent _returnvalue (delegate Custom|EndSkydiving))
      (EventDispatchers|BindEventtoOnEndParachuteMovementEvent _returnvalue (delegate Custom|EndParachute)))
    (:then_1
      (Components|Activation|Activate (Variables|Default|GetP_Backpack_MED_Mechanical_Engineer_Eyes) true))
    (:then_2
      (Components|Activation|Deactivate (Variables|Default|GetFX_AirTrail_L)))))

(event Custom|EndSkydiving
  ;[1 1 1] The skydive event,
  (Development|PrintString "END" true true "(R=1.000000,G=0.000000,B=0.311862,A=1.000000)")
  (Components|Activation|Activate (Variables|Default|GetP_Backpack_MED_Mechanical_Engineer_Eyes) true)
  (Components|Activation|Deactivate (Variables|Default|GetFX_AirTrail_L))
  (Utilities|FlowControl|Delay 1.5)
  (Components|Activation|Activate (Variables|Default|GetP_Backpack_MED_Mechanical_Engineer_SparkBurst) true))

(event Custom|EndParachute
  ;[1 1 1] The skydive event,
  (Development|PrintString "END" true true "(R=1.000000,G=0.000000,B=0.311862,A=1.000000)")
  (Components|Activation|Activate (Variables|Default|GetP_Backpack_MED_Mechanical_Engineer_Eyes) true)
  (Components|Activation|Deactivate (Variables|Default|GetFX_AirTrail_L))
  (Utilities|FlowControl|Delay 1.5)
  (Components|Activation|Activate (Variables|Default|GetP_Backpack_MED_Mechanical_Engineer_SparkBurst) true))

(event Custom|KillEvent
  ;[1 1 1] Kill event
  (Components|Activation|Activate (Variables|Default|GetFX_KillFlash) true)
  (Development|PrintString "Hello" true true "(R=0.102079,G=0.000000,B=1.000000,A=1.000000)"))

(fn ConstructionScript)
