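# Animation layer blueprint for the Mechanical Engineer Owl pet, covering skydive glide, dive, wrench detach and landing on the backpack.
# PetFullOverrideLayer selects a landing or skydive state machine, then stacks turn, lean, wing, feather and drift additives into a cached pose.
# World-space bone modifiers place or scale-hide the owl, and a glider pose snapshot blends in before the bUseLayer switch.
# Notifies update landing, detach and attach state, while PetSkeletalControlLayer adds head look-at.
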AnimBlueprint /BRCosmetics/Characters/Pets/Mechanical_Engineer_Owl/MechanicalOwlAnimLayer
target_skeleton=/BRCosmetics/Gadgets/Assets/VinderTech_GliderChute/Glider_Mechanical_Engineer_Female/Meshes/Mechanical_Engineer_Female_Glider_Skeleton
graphs [
 AnimGraph {nodes [Output Pose() {layerGroup=None}]}
 EventGraph {nodes [
   AnimNotify_DetachWrenchOff() {is=title K2Node_Event}
   > Set bDetachFromWrenchTransition(bDetachFromWrenchTransition:bool=false) -> (Output_Get:bool)
   AnimNotify_TheOwlHasLanded() {is=title K2Node_Event}
   > Set bCanPlayLandingAnim(bCanPlayLandingAnim:bool=false) -> (Output_Get:bool)
   > Set bOwlIsOnPack(bOwlIsOnPack:bool=true) -> (Output_Get:bool)
   AnimNotify_EnteredMechOwlSkydive() {is=title K2Node_Event}
   > Set AttachLocation(AttachLocation:Vector=Get PlayerRootLocation() -> (PlayerRootLocation:Vector)) -> (Output_Get:Vector)
   > Set OwlLocationOffset(OwlLocationOffset:Vector="0,150,100") -> (Output_Get:Vector)
  ]}
 PetFullOverrideLayer {nodes [
   n1 = "Transform (Modify) Bone - Bone: Glider_Base"(
    ComponentPose=Local To Component(LocalPose=Use cached pose 'PreJoints'())
    Alpha=1
    Translation=Get AttachLocation() -> (AttachLocation:Vector)
    Rotation=Get AttachRotation() -> (AttachRotation:Rotator)
    Scale="1,1,1"
   ) {
    rotationMode=BMM_Replace
    rotationSpace=BCS_WorldSpace
    translationMode=BMM_Replace
    translationSpace=BCS_WorldSpace
    alphaBoolBlend {blendOption=Linear}
    boneToModify {boneName=Glider_Base}
   }
   n2 = "Transform (Modify) Bone - Bone: Glider_Base"(
    ComponentPose=@n1
    Alpha=1
    Translation=Get OwlLocationOffset() -> (OwlLocationOffset:Vector)
    Rotation=Get OwlRotationOffset() -> (OwlRotationOffset:Rotator)
    Scale="1,1,1"
   ) {
    rotationMode=BMM_Additive
    rotationSpace=BCS_BoneSpace
    translationMode=BMM_Additive
    translationSpace=BCS_BoneSpace
    alphaBoolBlend {blendOption=Linear}
    boneToModify {boneName=Glider_Base}
   }
   n3 = "Transform (Modify) Bone - Bone: Frame"!(ComponentPose=@n2, bAlphaBoolEnabled:bool=True, Translation="0,0,0", Rotation="0,0,0", Scale="0.6,0.6,0.6") -> (Pose) {alphaInputType=Bool, scaleMode=BMM_Replace, alphaBoolBlend {blendOption=Linear}, boneToModify {boneName=Frame}, scale=<0.6 0.6 0.6>}
   Blend Poses by bool(
    BlendPose_0=Component To Local(ComponentPose=@n3)
    BlendTime_0=0
    BlendTime_1=0
    bActiveValue=Get bIsFrontEndContrails() -> (bIsFrontEndContrails:bool)
   ) {blendType=Linear}
   Local Space Ref Pose()
   n4 = "Transform (Modify) Bone - Bone: Root"(
    ComponentPose=Local To Component(LocalPose=Use cached pose 'PreJoints'())
    Alpha=1
    Translation=Get AttachLocation()
    Rotation=Get AttachRotation()
    Scale="1,1,1"
   ) {
    rotationMode=BMM_Replace
    rotationSpace=BCS_WorldSpace
    translationMode=BMM_Replace
    translationSpace=BCS_WorldSpace
    alphaBoolBlend {blendOption=Linear}
    boneToModify {boneName=Root}
   }
   n5 = "Transform (Modify) Bone - Bone: Glider_Base"(
    ComponentPose=@n4
    Alpha=Get Closed_NewCurveValue() -> (Closed_NewCurveValue:float)
    Translation=Get Owl_GLBase_Location() -> (Owl_GLBase_Location:Vector)
    Rotation=Get Owl_GLBase_Rotation() -> (Owl_GLBase_Rotation:Rotator)
    Scale="1,1,1"
   ) {
    rotationMode=BMM_Replace
    rotationSpace=BCS_WorldSpace
    translationMode=BMM_Replace
    translationSpace=BCS_WorldSpace
    alphaBoolBlend {blendOption=Linear}
    boneToModify {boneName=Glider_Base}
   }
   n6 = "Transform (Modify) Bone - Bone: Frame"!(
    ComponentPose=@n5
    Alpha:float=Get GliderCurveValue() -> (GliderCurveValue:float)
    Translation=Get GliderOwlFrameLandingStart() -> (GliderOwlFrameLandingStart:Vector)
    Rotation="0,0,0"
    Scale="1,1,1"
   ) -> (Pose) {
    rotationSpace=BCS_WorldSpace
    translationMode=BMM_Replace
    translationSpace=BCS_WorldSpace
    alphaBoolBlend {blendOption=Linear}
    boneToModify {boneName=Frame}
   }
   n7 = "Transform (Modify) Bone - Bone: Frame"!(
    ComponentPose=@n6
    bAlphaBoolEnabled:bool=Get bHideOwl() -> (bHideOwl:bool)
    Translation="0,0,0"
    Rotation="0,0,0"
    Scale="0.001,0.001,0.001"
   ) -> (Pose) {alphaInputType=Bool, scaleMode=BMM_Replace, alphaBoolBlend {blendOption=Linear}, boneToModify {boneName=Frame}, scale=<0.001 0.001 0.001>}
   n8 = "Transform (Modify) Bone - Bone: Glider_Base"(
    ComponentPose=Local To Component(LocalPose=Use cached pose 'PreJoints'())
    Alpha=Get ClosedAlpha() -> (ClosedAlpha:float)
    Translation=Get AttachLocation()
    Rotation=Get AttachRotation()
    Scale="1,1,1"
   ) {
    rotationMode=BMM_Replace
    rotationSpace=BCS_WorldSpace
    translationMode=BMM_Replace
    translationSpace=BCS_WorldSpace
    alphaBoolBlend {blendOption=Linear}
    boneToModify {boneName=Glider_Base}
   }
   n9 = "Transform (Modify) Bone - Bone: Glider_Base"(ComponentPose=@n8, Alpha=Get ClosedAlpha(), Translation=Get OwlLocationOffset(), Rotation="0,0,0", Scale="1,1,1") {
    rotationMode=BMM_Additive
    rotationSpace=BCS_WorldSpace
    translationMode=BMM_Additive
    translationSpace=BCS_BoneSpace
    alphaBoolBlend {blendOption=Linear}
    boneToModify {boneName=Glider_Base}
   }
   n10 = "Transform (Modify) Bone - Bone: Frame"(ComponentPose=@n9, bAlphaBoolEnabled=Get bHideOwl(), Translation="0,0,0", Rotation="0,0,0", Scale="0.001,0.001,0.001") {alphaInputType=Bool, scaleMode=BMM_Replace, alphaBoolBlend {blendOption=Linear}, boneToModify {boneName=Frame}, scale=<0.001 0.001 0.001>}
   n11 = "Transform (Modify) Bone - Bone: Frame"!(
    ComponentPose=@n10
    Alpha:float=Get OwlFrameAlpha() -> (OwlFrameAlpha:float)
    Translation=Get PlayerRootLocation()
    Rotation=Get PlayerRootRotation() -> (PlayerRootRotation:Rotator)
    Scale="1,1,1"
   ) -> (Pose) {
    rotationSpace=BCS_WorldSpace
    translationMode=BMM_Replace
    translationSpace=BCS_WorldSpace
    alphaBoolBlend {blendOption=Linear}
    boneToModify {boneName=Frame}
   }
   n12 = "Transform (Modify) Bone - Bone: Frame"(
    ComponentPose=@n11
    Alpha=Get OwlFrameAlpha()
    Translation=Get GliderOwlSnapshotLocation() -> (GliderOwlSnapshotLocation:Vector)
    Rotation=Get GliderOwlSnapshotRotation() -> (GliderOwlSnapshotRotation:Rotator)
    Scale="1,1,1"
   ) {
    rotationMode=BMM_Replace
    rotationSpace=BCS_WorldSpace
    translationMode=BMM_Additive
    translationSpace=BCS_WorldSpace
    alphaBoolBlend {blendOption=Linear}
    boneToModify {boneName=Frame}
   }
   n13 = Blend Poses by bool(
    BlendPose_0=Component To Local(ComponentPose=@n7)
    BlendPose_1=Component To Local(ComponentPose=@n12)
    BlendTime_0=0
    BlendTime_1=0
    bActiveValue=Get bIsInLandingPhase() -> (bIsInLandingPhase:bool)
   ) {blendType=Linear}
   n14 = Pose Snapshot!(Snapshot:Pose Snapshot Structure=Get GliderSnapshotPose() -> (GliderSnapshotPose:Pose Snapshot Structure)) -> (Pose:Pose Link Structure) {is=title AnimGraphNode_PoseSnapshot, mode=SnapshotPin}
   n15 = Get OwlPoseAlpha() -> (OwlPoseAlpha:float)
   Output Pose(Result=Blend Poses by bool(
     BlendPose_0=Layered blend per bone(BasePose=@n13, BlendPoses_0=@n14, BlendWeights_0=@n15) {bBlendRootMotionBasedOnRootBone=false, layerSetup [{branchFilters [{boneName=pelvis}]}]}
     BlendPose_1=Input Pose()
     BlendTime_0=0
     BlendTime_1=0
     bActiveValue=Get bUseLayer() -> (bUseLayer:bool)
    ) {blendType=Linear}) {layerGroup=PetOverride}
   n16 = Get SkydivingState() -> (SkydivingState:Fort Anim Input Skydiving Structure)
   n17 = Get AimSmooth() -> (AimSmooth:Fort Anim Input Aim Scramble Data Helper Structure)
   n18 = Apply Additive(
    Base=Use cached pose 'PreAdditives'()
    Additive=Blendspace Player 'BS_MechOwl_Turn'(
     X=Break Fort Anim Input Aim Scramble Data Helper!(FortAnimInput_AimScrambleDataHelper:"Fort Anim Input Aim Scramble Data Helper Structure (by ref)"=@n17) -> (AimPitch:float=~, AimYaw:float=~) {is=title K2Node_BreakStruct}.AimYaw
     Y=Break Fort Anim Input Skydiving!(FortAnimInput_Skydiving:"Fort Anim Input Skydiving Structure (by ref)"=@n16) -> (
      ParaGilderRootOffset:Vector=~
      ParaGliderRootRotationOverride:Rotator=~
      bUseParaGlideRootModifier:bool=~
      bIsSkydivingFromLaunchPad:bool=~
      bIsSkydivingFromBus:bool=~
      bIsInVortex:bool=~
      bIsDiving:bool=~
      bIsParachuteOpen:bool=~
      bIsSkydiving:bool=~
      bIsParachuteLeaning:bool=~
      bIsSkydiveLeaning:bool=~
      bIsLeaning:bool=~
      bIsSkydiveDiveMode:bool=~
      bPlaySkydiveDrift:bool=~
      bSkydiveDriftDelayActive:bool=~
      bSkydiveDriftAnimAllowed:bool=~
      bIsGliderRight:bool=~
      bIsGliderCenter:bool=~
      bIsGliderLeft:bool=~
      bIsGliderForward:bool=~
      bIsGliderBack:bool=~
      bHasDeployedAndIsStandingUpright:bool=~
      LocalAccelForward:float=~
      LocalAccelRight:float=~
      LocalVelocityRight:float=~
      SkydiveAimPitch:float=~
      SkydiveAimYaw:float=~
      DeployChuteAnimRate:float=~
      SkydiveDriftAnimRate:float=~
      SkydiveDriftAnimRateCurrent:float=~
      SkydiveFidgetAnimRate:float=~
      SkydiveFidgetAnimRateCurrent:float=~
      SkydiveAdditiveAlpha:float=~
      SkydiveDriftDelay:float=~
      SkydiveDriftAnim:int
      SkydiveDriftAnimMax:int
      LaunchpadAnim:int
      LocalAccelDir:ESkydivingDirection Enum
      DirectionLast:ESkydivingDirection Enum
      PoseAdditive:Anim Sequence Object Reference
     ) {is=title K2Node_BreakStruct}.SkydiveAimPitch
    ) {blendSpace=/BRCosmetics/Animation/Game/Pets/MechOwl/Backpack/Skydive/BS_MechOwl_Turn}
    Alpha=Get AdditiveAlpha() -> (AdditiveAlpha:float)
   ) {alphaBoolBlend {blendOption=Linear}}
   n19 = Break Fort Anim Input Skydiving(FortAnimInput_Skydiving=Get SkydivingState()) -> (
    ParaGilderRootOffset=~
    ParaGliderRootRotationOverride=~
    bUseParaGlideRootModifier=~
    bIsSkydivingFromLaunchPad=~
    bIsSkydivingFromBus=~
    bIsInVortex=~
    bIsDiving=~
    bIsParachuteOpen=~
    bIsSkydiving=~
    bIsParachuteLeaning=~
    bIsSkydiveLeaning=~
    bIsLeaning=~
    bIsSkydiveDiveMode=~
    bPlaySkydiveDrift=~
    bSkydiveDriftDelayActive=~
    bSkydiveDriftAnimAllowed=~
    bIsGliderRight=~
    bIsGliderCenter=~
    bIsGliderLeft=~
    bIsGliderForward=~
    bIsGliderBack=~
    bHasDeployedAndIsStandingUpright=~
    LocalAccelForward=~
    LocalAccelRight=~
    LocalVelocityRight=~
    SkydiveAimPitch=~
    SkydiveAimYaw=~
    DeployChuteAnimRate=~
    SkydiveDriftAnimRate=~
    SkydiveDriftAnimRateCurrent=~
    SkydiveFidgetAnimRate=~
    SkydiveFidgetAnimRateCurrent=~
    SkydiveAdditiveAlpha=~
    SkydiveDriftDelay=~
   )
   n20 = Blendspace Player 'BS_MechOwl_Lean'(X=@n19.LocalVelocityRight, Y=@n19.SkydiveAimPitch) {blendSpace=/BRCosmetics/Animation/Game/Pets/MechOwl/Backpack/Skydive/BS_MechOwl_Lean}
   n21 = Get AdditiveAlpha()
   n22 = Random Sequence Player!() -> (Pose:Pose Link Structure) {
    is=title AnimGraphNode_RandomPlayer
    bShuffleMode=true
    entries [
     {
      chanceToPlay=1
      maxLoopCount=0
      maxPlayRate=1.25
      minLoopCount=0
      minPlayRate=1
      sequence=/BRCosmetics/Animation/Game/Pets/MechOwl/Backpack/Skydive/MechOwl_Glide_WingSway
      blendIn {blendOption=Linear, blendTime=0.2, customCurve=null}
     }
     {
      chanceToPlay=1
      maxLoopCount=0
      maxPlayRate=1.25
      minLoopCount=0
      minPlayRate=1
      sequence=/BRCosmetics/Animation/Game/Pets/MechOwl/Backpack/Skydive/MechOwl_Glide_WingSway_B
      blendIn {blendOption=Linear, blendTime=0.2, customCurve=null}
     }
     {
      chanceToPlay=1
      maxLoopCount=0
      maxPlayRate=1.45
      minLoopCount=0
      minPlayRate=1.25
      sequence=&MechOwl_Glide_WingSway_B
      blendIn {blendOption=Linear, blendTime=0.2, customCurve=null}
     }
    ]
   }
   n23 = Apply Additive(
    Base=Apply Additive(Base=Apply Additive(Base=@n18, Additive=@n20, Alpha=@n21) {alphaBoolBlend {blendOption=Linear}}, Additive=@n22, Alpha=Get AdditiveAlpha()) {alphaBoolBlend {blendOption=Linear}}
    Additive=Sequence Player 'MechOwl_Glide_Feathers'() {sequence=/BRCosmetics/Animation/Game/Pets/MechOwl/Backpack/Skydive/MechOwl_Glide_Feathers}
    Alpha=Get AdditiveAlpha()
   ) {alphaBoolBlend {blendOption=Linear}}
   n24 = Random Sequence Player() {entries [
     {
      chanceToPlay=1
      maxLoopCount=0
      maxPlayRate=1
      minLoopCount=0
      minPlayRate=1
      sequence=/BRCosmetics/Animation/Game/Pets/MechOwl/Backpack/Skydive/MechOwl_Glide_DriftA
      blendIn {blendOption=Linear, blendTime=0.2, customCurve=null}
     }
     {
      chanceToPlay=1
      maxLoopCount=0
      maxPlayRate=1
      minLoopCount=0
      minPlayRate=1
      sequence=/BRCosmetics/Animation/Game/Pets/MechOwl/Backpack/Skydive/MechOwl_Glide_DriftB
      blendIn {blendOption=Linear, blendTime=0.2, customCurve=null}
     }
     {
      chanceToPlay=1
      maxLoopCount=0
      maxPlayRate=1
      minLoopCount=0
      minPlayRate=1
      sequence=/BRCosmetics/Animation/Game/Pets/MechOwl/Backpack/Skydive/MechOwl_Glide_DriftC
      blendIn {blendOption=Linear, blendTime=0.2, customCurve=null}
     }
    ]}
   n25 = Get AdditiveAlpha()
   Save cached pose 'PreJoints'(Pose=Apply Additive(
     Base=Apply Additive(Base=@n23, Additive=@n24, Alpha=@n25) {alphaBoolBlend {blendOption=Linear}}
     Additive=Sequence Player 'MechOwl_Glide_Pose_Add'() {sequence=/BRCosmetics/Animation/Game/Pets/MechOwl/Backpack/Skydive/MechOwl_Glide_Pose_Add}
     Alpha=Get WingCorrectiveAdditiveAlpha() -> (WingCorrectiveAdditiveAlpha:float)
    ) {alphaBoolBlend {blendOption=Linear}, alphaScaleBiasClamp {outRange {max=0, min=1}}}) {cacheName=PreJoints}
   AnimGraphNode_StateMachine_0 = LandingMachine() {is=title AnimGraphNode_StateMachine}
   AnimGraphNode_StateMachine_1 = SkyDiveMachine() {is=title AnimGraphNode_StateMachine}
   Save cached pose 'PreAdditives'(Pose=Blend Poses by bool(
     BlendPose_0=@AnimGraphNode_StateMachine_0
     BlendPose_1=@AnimGraphNode_StateMachine_1
     BlendTime_0=0
     BlendTime_1=0
     bActiveValue=Get bCanPlayLandingAnim() -> (bCanPlayLandingAnim:bool)
    ) {blendType=Linear}) {cacheName=PreAdditives}
   Sequence Player 'Idle_Pose'() {sequence=/BRCosmetics/Animation/Game/Pets/MechOwl/Backpack/Idle/Idle_Pose}
   Sequence Player 'MechOwl_Glide_WingSway'() {sequence=&MechOwl_Glide_WingSway}
   Use cached pose 'PreJoints'()
   Get GliderOwlSnapshotLocation()
   Get GliderOwlSnapshotRotation()
  ]}
 LandingMachine {
  entry=to_Closed
  path=PetFullOverrideLayer.AnimGraphNode_StateMachine_0.LandingMachine
  ids=id
  nodes [
   AnimStateNode_0 = to_Closed() {is=title AnimStateNode}
   AnimStateNode_1 = Closed() {is=title AnimStateNode}
   AnimStateTransitionNode_0 = AnimStateTransitionNode() {
    bAutomaticRuleBasedOnSequencePlayerInState=true
    crossfadeDuration=0
    transitionStart {guid=#1, notifyName=TheOwlHasLanded}
   }
  ]
  links [{
    dst=Closed
    rule=PetFullOverrideLayer.AnimGraphNode_StateMachine_0.LandingMachine.AnimStateTransitionNode_0.AnimationTransitionGraph_1
    src=to_Closed
    via=AnimStateTransitionNode_0
   }]
 }
 to_Closed {
  path=PetFullOverrideLayer.AnimGraphNode_StateMachine_0.LandingMachine.AnimStateNode_0.to_Closed
  nodes [Output Animation Pose(Result=Sequence Player 'MechOwl_Open_to_Backpack'() {bLoopAnimation=false, sequence=/BRCosmetics/Animation/Game/Pets/MechOwl/Backpack/Skydive/Transitions/MechOwl_Open_to_Backpack})]
 }
 Closed {
  path=PetFullOverrideLayer.AnimGraphNode_StateMachine_0.LandingMachine.AnimStateNode_1.Closed
  nodes [
   Sequence Player 'MechOwl_Closed_Pose'() {sequence=/BRCosmetics/Animation/Game/Pets/MechOwl/Backpack/Skydive/MechOwl_Closed_Pose}
   Sequence Player 'MechOwl_Open_to_Backpack_Closed_Pose'() {sequence=/BRCosmetics/Animation/Game/Pets/MechOwl/Backpack/Skydive/Transitions/MechOwl_Open_to_Backpack_Closed_Pose}
   n1 = Random Sequence Player() {
    bShuffleMode=true
    entries [
     {
      chanceToPlay=1
      maxLoopCount=0
      maxPlayRate=1
      minLoopCount=0
      minPlayRate=1
      sequence=/BRCosmetics/Animation/Game/Pets/MechOwl/Backpack/Idle/Idle
      blendIn {blendOption=Linear, blendTime=0.2, customCurve=null}
     }
     {
      chanceToPlay=1
      maxLoopCount=0
      maxPlayRate=1
      minLoopCount=0
      minPlayRate=1
      sequence=/BRCosmetics/Animation/Game/Pets/MechOwl/Backpack/Idle/IdleBreak_01
      blendIn {blendOption=Linear, blendTime=0.2, customCurve=null}
     }
     {
      chanceToPlay=1
      maxLoopCount=0
      maxPlayRate=1
      minLoopCount=0
      minPlayRate=1
      sequence=/BRCosmetics/Animation/Game/Pets/MechOwl/Backpack/Idle/IdleBreak_02
      blendIn {blendOption=Linear, blendTime=0.2, customCurve=null}
     }
    ]
   }
   Output Animation Pose(Result=@n1)
  ]
 }
 AnimationTransitionGraph_1 {path=PetFullOverrideLayer.AnimGraphNode_StateMachine_0.LandingMachine.AnimStateTransitionNode_0.AnimationTransitionGraph_1, nodes [Result()]}
 SkyDiveMachine {
  entry=Closed
  path=PetFullOverrideLayer.AnimGraphNode_StateMachine_1.SkyDiveMachine
  ids=id
  nodes [
   To_Closed_Alias() {is=title AnimStateAliasNode, also {aliases [Dive,Glide,to_Dive,to_Glide,to_Skydive]}}
   Detach_Alias() {is=title AnimStateAliasNode, also {aliases [Dive,Glide,to_Dive,to_Glide]}}
   AnimStateNode_10 = Dive() {is=title AnimStateNode}
   AnimStateNode_2 = Passthrough() {is=title AnimStateNode}
   AnimStateNode_3 = to_Dive() {is=title AnimStateNode}
   AnimStateNode_4 = to_Glide() {is=title AnimStateNode}
   AnimStateNode_5 = Detach() {is=title AnimStateNode}
   AnimStateNode_6 = to_Closed() {
    stateEntered {guid=#2}
    stateFullyBlended {guid=#3}
    stateLeft {guid=#4}
   }
   AnimStateNode_7 = to_Skydive() {
    is=title AnimStateNode
    stateEntered {guid=#5, notifyName=EnteredMechOwlSkydive}
    stateFullyBlended {guid=#6}
    stateLeft {guid=#7}
   }
   AnimStateNode_8 = Closed()
   AnimStateNode_9 = Glide() {is=title AnimStateNode}
   AnimStateTransitionNode_10 = AnimStateTransitionNode() {bAutomaticRuleBasedOnSequencePlayerInState=true}
   AnimStateTransitionNode_15 = AnimStateTransitionNode() {crossfadeDuration=0.445}
   AnimStateTransitionNode_16 = AnimStateTransitionNode() {crossfadeDuration=0}
   AnimStateTransitionNode_18 = AnimStateTransitionNode() {crossfadeDuration=0}
   AnimStateTransitionNode_19 = AnimStateTransitionNode() {crossfadeDuration=0}
   AnimStateTransitionNode_2 = AnimStateTransitionNode() {crossfadeDuration=0.15}
   AnimStateTransitionNode_21 = AnimStateTransitionNode()
   AnimStateTransitionNode_24 = AnimStateTransitionNode() {crossfadeDuration=0}
   AnimStateTransitionNode_3 = AnimStateTransitionNode() {crossfadeDuration=0.15}
   AnimStateTransitionNode_4 = AnimStateTransitionNode() {bAutomaticRuleBasedOnSequencePlayerInState=true, crossfadeDuration=0}
   AnimStateTransitionNode_5 = AnimStateTransitionNode() {bAutomaticRuleBasedOnSequencePlayerInState=true, crossfadeDuration=0}
   AnimStateTransitionNode_6 = AnimStateTransitionNode()
   AnimStateTransitionNode_7 = AnimStateTransitionNode()
   AnimStateTransitionNode_8 = AnimStateTransitionNode() {
    bAutomaticRuleBasedOnSequencePlayerInState=true
    crossfadeDuration=0
    transitionStart {guid=#8, notifyName=TheOwlHasLanded}
   }
   AnimStateTransitionNode_9 = AnimStateTransitionNode() {crossfadeDuration=0}
  ]
  links [
   {
    dst=to_Skydive
    rule=PetFullOverrideLayer.AnimGraphNode_StateMachine_1.SkyDiveMachine.AnimStateTransitionNode_9.AnimationTransitionGraph_0
    src=Closed
    via=AnimStateTransitionNode_9
   }
   {
    dst=Dive
    rule=PetFullOverrideLayer.AnimGraphNode_StateMachine_1.SkyDiveMachine.AnimStateTransitionNode_19.AnimationTransitionGraph_1
    src=Detach
    via=AnimStateTransitionNode_19
   }
   {
    dst=Glide
    rule=PetFullOverrideLayer.AnimGraphNode_StateMachine_1.SkyDiveMachine.AnimStateTransitionNode_18.AnimationTransitionGraph_1
    src=Detach
    via=AnimStateTransitionNode_18
   }
   {
    dst=Detach
    rule=PetFullOverrideLayer.AnimGraphNode_StateMachine_1.SkyDiveMachine.AnimStateTransitionNode_21.Transition
    src=Detach_Alias
    via=AnimStateTransitionNode_21
   }
   {
    dst=to_Glide
    rule=PetFullOverrideLayer.AnimGraphNode_StateMachine_1.SkyDiveMachine.AnimStateTransitionNode_3.AnimationTransitionGraph_1
    src=Dive
    via=AnimStateTransitionNode_3
   }
   {
    dst=to_Dive
    rule=PetFullOverrideLayer.AnimGraphNode_StateMachine_1.SkyDiveMachine.AnimStateTransitionNode_2.AnimationTransitionGraph_1
    src=Glide
    via=AnimStateTransitionNode_2
   }
   {
    dst=Glide
    rule=PetFullOverrideLayer.AnimGraphNode_StateMachine_1.SkyDiveMachine.AnimStateTransitionNode_24.AnimationTransitionGraph_1
    src=Passthrough
    via=AnimStateTransitionNode_24
   }
   {
    dst=to_Closed
    rule=PetFullOverrideLayer.AnimGraphNode_StateMachine_1.SkyDiveMachine.AnimStateTransitionNode_16.Transition
    src=To_Closed_Alias
    via=AnimStateTransitionNode_16
   }
   {
    dst=Closed
    rule=PetFullOverrideLayer.AnimGraphNode_StateMachine_1.SkyDiveMachine.AnimStateTransitionNode_8.AnimationTransitionGraph_0
    src=to_Closed
    via=AnimStateTransitionNode_8
   }
   {
    dst=to_Skydive
    rule=PetFullOverrideLayer.AnimGraphNode_StateMachine_1.SkyDiveMachine.AnimStateTransitionNode_15.Transition
    src=to_Closed
    via=AnimStateTransitionNode_15
   }
   {
    dst=Dive
    rule=PetFullOverrideLayer.AnimGraphNode_StateMachine_1.SkyDiveMachine.AnimStateTransitionNode_4.AnimationTransitionGraph_1
    src=to_Dive
    via=AnimStateTransitionNode_4
   }
   {
    dst=to_Glide
    rule=PetFullOverrideLayer.AnimGraphNode_StateMachine_1.SkyDiveMachine.AnimStateTransitionNode_7.AnimationTransitionGraph_1
    src=to_Dive
    via=AnimStateTransitionNode_7
   }
   {
    dst=Glide
    rule=PetFullOverrideLayer.AnimGraphNode_StateMachine_1.SkyDiveMachine.AnimStateTransitionNode_5.AnimationTransitionGraph_1
    src=to_Glide
    via=AnimStateTransitionNode_5
   }
   {
    dst=to_Dive
    rule=PetFullOverrideLayer.AnimGraphNode_StateMachine_1.SkyDiveMachine.AnimStateTransitionNode_6.AnimationTransitionGraph_1
    src=to_Glide
    via=AnimStateTransitionNode_6
   }
   {
    dst=Glide
    rule=PetFullOverrideLayer.AnimGraphNode_StateMachine_1.SkyDiveMachine.AnimStateTransitionNode_10.AnimationTransitionGraph_0
    src=to_Skydive
    via=AnimStateTransitionNode_10
   }
  ]
 }
 Dive {
  path=PetFullOverrideLayer.AnimGraphNode_StateMachine_1.SkyDiveMachine.AnimStateNode_10.Dive
  nodes [
   n1 = Get SkydivingState()
   n2 = Get AimSmooth()
   Output Animation Pose(Result=Blendspace Player 'BS_MechOwl_Dive'(
     X=Break Fort Anim Input Aim Scramble Data Helper(FortAnimInput_AimScrambleDataHelper=@n2) -> (AimPitch=~, AimYaw=~).AimYaw
     Y=Break Fort Anim Input Skydiving!(FortAnimInput_Skydiving=@n1) -> (
      bIsGliderRight=~
      bIsGliderCenter=~
      bIsGliderLeft=~
      bIsGliderForward=~
      bIsGliderBack=~
      bHasDeployedAndIsStandingUpright=~
      SkydiveAimPitch=~
      SkydiveAimYaw=~
      PoseAdditive
     ).SkydiveAimPitch
    ) {blendSpace=/BRCosmetics/Animation/Game/Pets/MechOwl/Backpack/Skydive/BS_MechOwl_Dive})
  ]
 }
 Passthrough {
  path=PetFullOverrideLayer.AnimGraphNode_StateMachine_1.SkyDiveMachine.AnimStateNode_2.Passthrough
  nodes [Output Animation Pose(Result=Local Space Ref Pose())]
 }
 to_Dive {
  path=PetFullOverrideLayer.AnimGraphNode_StateMachine_1.SkyDiveMachine.AnimStateNode_3.to_Dive
  nodes [
   n1 = Get SkydivingState()
   n2 = Get AimSmooth()
   Output Animation Pose(Result=Blendspace Player 'BS_MechOwl_Glide_To_Dive'(
     X=Break Fort Anim Input Aim Scramble Data Helper(FortAnimInput_AimScrambleDataHelper=@n2) -> (AimPitch=~, AimYaw=~).AimYaw
     Y=Break Fort Anim Input Skydiving(FortAnimInput_Skydiving=@n1) -> (
      bIsGliderRight=~
      bIsGliderCenter=~
      bIsGliderLeft=~
      bIsGliderForward=~
      bIsGliderBack=~
      bHasDeployedAndIsStandingUpright=~
      SkydiveAimPitch=~
      SkydiveAimYaw=~
     ).SkydiveAimPitch
    ) {bLoop=false, blendSpace=/BRCosmetics/Animation/Game/Pets/MechOwl/Backpack/Skydive/BS_MechOwl_Glide_To_Dive})
  ]
 }
 to_Glide {
  path=PetFullOverrideLayer.AnimGraphNode_StateMachine_1.SkyDiveMachine.AnimStateNode_4.to_Glide
  nodes [
   n1 = Get SkydivingState()
   n2 = Get AimSmooth()
   Output Animation Pose(Result=Blendspace Player 'BS_MechOwl_Dive_To_Glide'(
     X=Break Fort Anim Input Aim Scramble Data Helper(FortAnimInput_AimScrambleDataHelper=@n2) -> (AimPitch=~, AimYaw=~).AimYaw
     Y=Break Fort Anim Input Skydiving(FortAnimInput_Skydiving=@n1) -> (
      bIsGliderRight=~
      bIsGliderCenter=~
      bIsGliderLeft=~
      bIsGliderForward=~
      bIsGliderBack=~
      bHasDeployedAndIsStandingUpright=~
      SkydiveAimPitch=~
      SkydiveAimYaw=~
     ).SkydiveAimPitch
    ) {bLoop=false, blendSpace=/BRCosmetics/Animation/Game/Pets/MechOwl/Backpack/Skydive/BS_MechOwl_Dive_To_Glide})
  ]
 }
 Detach {
  path=PetFullOverrideLayer.AnimGraphNode_StateMachine_1.SkyDiveMachine.AnimStateNode_5.Detach
  nodes [
   n1 = Break Fort Anim Input Skydiving(FortAnimInput_Skydiving=Get SkydivingState()) -> (
    bIsGliderRight=~
    bIsGliderCenter=~
    bIsGliderLeft=~
    bIsGliderForward=~
    bIsGliderBack=~
    bHasDeployedAndIsStandingUpright=~
    SkydiveAimPitch=~
    SkydiveAimYaw=~
   )
   n2 = Blendspace Player 'BS_MechOwl_DetachDive'(X=Get SkydiveForward() -> (SkydiveForward:float), Y=@n1.SkydiveAimPitch) {
    bLoop=false
    blendSpace=/BRCosmetics/Animation/Game/Pets/MechOwl/Backpack/Skydive/Transitions/BS_MechOwl_DetachDive
    groupName=detach
    method=SyncGroup
   }
   n3 = Blendspace Player 'BS_MechOwl_DetachGlide'(X=Get SkydiveForward(), Y=@n1.SkydiveAimPitch) {
    bLoop=false
    blendSpace=/BRCosmetics/Animation/Game/Pets/MechOwl/Backpack/Skydive/Transitions/BS_MechOwl_DetachGlide
    groupName=detach
    method=SyncGroup
   }
   n4 = Get SkydivingState()
   Output Animation Pose(Result=Blend Poses by bool(
     BlendPose_0=@n2
     BlendPose_1=@n3
     bActiveValue=Break Fort Anim Input Skydiving!(FortAnimInput_Skydiving=@n4) -> (
      bIsSkydiveDiveMode:bool=~
      bIsGliderRight=~
      bIsGliderCenter=~
      bIsGliderLeft=~
      bIsGliderForward=~
      bIsGliderBack=~
      bHasDeployedAndIsStandingUpright=~
      PoseAdditive
     )
    ) {blendType=Linear})
  ]
 }
 to_Closed {
  path=PetFullOverrideLayer.AnimGraphNode_StateMachine_1.SkyDiveMachine.AnimStateNode_6.to_Closed
  nodes [
   Blend Poses by bool(
    BlendPose_1=Blend Poses by bool(
     BlendPose_1=Sequence Player 'MechOwl_Open_to_Backpack'() {bLoopAnimation=false, sequence=&MechOwl_Open_to_Backpack}
     bActiveValue=Get bUsingUmbrellaGliderSet() -> (bUsingUmbrellaGliderSet:bool)
    ) {blendType=Linear}
    bActiveValue=Get bUsingHangGliderSet() -> (bUsingHangGliderSet:bool)
   ) {blendType=Linear}
   Output Animation Pose(Result=Sequence Player 'MechOwl_Open_to_Backpack'() {bLoopAnimation=false, sequence=&MechOwl_Open_to_Backpack})
   "Equal (SoftObjectReference)"!(
    A:"Soft Object Reference (by ref)"=Get GliderPlayerAnimSet() -> (GliderPlayerAnimSet:Fort Player Glider Anim Set Soft Object Reference)
    B:"Soft Object Reference (by ref)"=Get UmbrellaAnimSet() -> (UmbrellaAnimSet:Fort Player Glider Anim Set Soft Object Reference)
   ) -> (ReturnValue:bool)
   "Equal (SoftObjectReference)"(A=Get GliderPlayerAnimSet(), B=Get HangGliderAnimSet() -> (HangGliderAnimSet:Fort Player Glider Anim Set Soft Object Reference))
  ]
 }
 to_Skydive {
  path=PetFullOverrideLayer.AnimGraphNode_StateMachine_1.SkyDiveMachine.AnimStateNode_7.to_Skydive
  nodes [Output Animation Pose(Result=Sequence Player 'MechOwl_Closed_to_Glide'() {bLoopAnimation=false, sequence=/BRCosmetics/Animation/Game/Pets/MechOwl/Backpack/Skydive/Transitions/MechOwl_Closed_to_Glide})]
 }
 Closed {
  path=PetFullOverrideLayer.AnimGraphNode_StateMachine_1.SkyDiveMachine.AnimStateNode_8.Closed
  nodes [
   Sequence Player 'MechOwl_Open_to_Backpack_Closed_Pose'() {sequence=&MechOwl_Open_to_Backpack_Closed_Pose}
   Sequence Player 'MechOwl_Closed_Pose'() {sequence=&MechOwl_Closed_Pose}
   n1 = Random Sequence Player() {
    bShuffleMode=true
    entries [
     {
      chanceToPlay=1
      maxLoopCount=0
      maxPlayRate=1
      minLoopCount=0
      minPlayRate=1
      sequence=&Idle
      blendIn {blendOption=Linear, blendTime=0.2, customCurve=null}
     }
     {
      chanceToPlay=1
      maxLoopCount=0
      maxPlayRate=1
      minLoopCount=0
      minPlayRate=1
      sequence=&IdleBreak_01
      blendIn {blendOption=Linear, blendTime=0.2, customCurve=null}
     }
     {
      chanceToPlay=1
      maxLoopCount=0
      maxPlayRate=1
      minLoopCount=0
      minPlayRate=1
      sequence=&IdleBreak_02
      blendIn {blendOption=Linear, blendTime=0.2, customCurve=null}
     }
    ]
   }
   Output Animation Pose(Result=@n1)
  ]
 }
 Glide {
  path=PetFullOverrideLayer.AnimGraphNode_StateMachine_1.SkyDiveMachine.AnimStateNode_9.Glide
  nodes [Output Animation Pose(Result=Sequence Player 'MechOwl_Glide_Pose'() {sequence=/BRCosmetics/Animation/Game/Pets/MechOwl/Backpack/Skydive/MechOwl_Glide_Pose})]
 }
 AnimationTransitionGraph_0 {
  path=PetFullOverrideLayer.AnimGraphNode_StateMachine_1.SkyDiveMachine.AnimStateTransitionNode_10.AnimationTransitionGraph_0
  nodes [Result()]
 }
 Transition {
  path=PetFullOverrideLayer.AnimGraphNode_StateMachine_1.SkyDiveMachine.AnimStateTransitionNode_15.Transition
  nodes [Result(bCanEnterTransition=Get bTransition_ToSkydive() -> (bTransition_ToSkydive:bool))]
 }
 Transition {
  path=PetFullOverrideLayer.AnimGraphNode_StateMachine_1.SkyDiveMachine.AnimStateTransitionNode_16.Transition
  nodes [Result(bCanEnterTransition=Get bTransition_ToClosed() -> (bTransition_ToClosed:bool))]
 }
 AnimationTransitionGraph_1 {
  path=PetFullOverrideLayer.AnimGraphNode_StateMachine_1.SkyDiveMachine.AnimStateTransitionNode_18.AnimationTransitionGraph_1
  nodes [Result(bCanEnterTransition=Get bTransition_Detach_To_Glide() -> (bTransition_Detach_To_Glide:bool))]
 }
 AnimationTransitionGraph_1 {
  path=PetFullOverrideLayer.AnimGraphNode_StateMachine_1.SkyDiveMachine.AnimStateTransitionNode_19.AnimationTransitionGraph_1
  nodes [Result(bCanEnterTransition=Get bTransition_Detach_To_Dive() -> (bTransition_Detach_To_Dive:bool))]
 }
 AnimationTransitionGraph_1 {
  path=PetFullOverrideLayer.AnimGraphNode_StateMachine_1.SkyDiveMachine.AnimStateTransitionNode_2.AnimationTransitionGraph_1
  nodes [Result(bCanEnterTransition=Break Fort Anim Input Skydiving!(FortAnimInput_Skydiving=Get SkydivingState()) -> (
     ParaGilderRootOffset:Vector=~
     ParaGliderRootRotationOverride:Rotator=~
     bUseParaGlideRootModifier:bool=~
     bIsSkydivingFromLaunchPad:bool=~
     bIsSkydivingFromBus:bool=~
     bIsInVortex:bool=~
     bIsDiving:bool=~
     bIsParachuteOpen:bool=~
     bIsSkydiving:bool=~
     bIsParachuteLeaning:bool=~
     bIsSkydiveLeaning:bool=~
     bIsLeaning:bool=~
     bIsSkydiveDiveMode=~
     bPlaySkydiveDrift:bool=~
     bSkydiveDriftDelayActive:bool=~
     bSkydiveDriftAnimAllowed:bool=~
     bIsGliderRight=~
     bIsGliderCenter=~
     bIsGliderLeft=~
     bIsGliderForward=~
     bIsGliderBack=~
     bHasDeployedAndIsStandingUpright=~
     LocalAccelForward:float=~
     LocalAccelRight:float=~
     LocalVelocityRight:float=~
     SkydiveAimPitch:float=~
     SkydiveAimYaw:float=~
     DeployChuteAnimRate:float=~
     SkydiveDriftAnimRate:float=~
     SkydiveDriftAnimRateCurrent:float=~
     SkydiveFidgetAnimRate:float=~
     SkydiveFidgetAnimRateCurrent:float=~
     SkydiveAdditiveAlpha:float=~
     SkydiveDriftDelay:float=~
     SkydiveDriftAnim:int
     SkydiveDriftAnimMax:int
     LaunchpadAnim:int
     LocalAccelDir:ESkydivingDirection Enum
     DirectionLast:ESkydivingDirection Enum
     PoseAdditive
    ).bIsSkydiveDiveMode)]
 }
 Transition {
  path=PetFullOverrideLayer.AnimGraphNode_StateMachine_1.SkyDiveMachine.AnimStateTransitionNode_21.Transition
  nodes [Result(bCanEnterTransition=Get bDetachFromWrenchTransition() -> (bDetachFromWrenchTransition:bool))]
 }
 AnimationTransitionGraph_1 {
  path=PetFullOverrideLayer.AnimGraphNode_StateMachine_1.SkyDiveMachine.AnimStateTransitionNode_24.AnimationTransitionGraph_1
  nodes [Result(bCanEnterTransition=Break Fort Anim Input Skydiving!(FortAnimInput_Skydiving=Get SkydivingState()) -> (
     bIsSkydiving=~
     bIsGliderRight=~
     bIsGliderCenter=~
     bIsGliderLeft=~
     bIsGliderForward=~
     bIsGliderBack=~
     bHasDeployedAndIsStandingUpright=~
     SkydiveAimPitch=~
     SkydiveAimYaw=~
     PoseAdditive
    ))]
 }
 AnimationTransitionGraph_1 {
  path=PetFullOverrideLayer.AnimGraphNode_StateMachine_1.SkyDiveMachine.AnimStateTransitionNode_3.AnimationTransitionGraph_1
  nodes [
   n1 = Get SkydivingState()
   Result(bCanEnterTransition=NOT Boolean(A=Break Fort Anim Input Skydiving!(FortAnimInput_Skydiving=@n1) -> (
      bIsSkydiveDiveMode:bool=~
      bIsGliderRight=~
      bIsGliderCenter=~
      bIsGliderLeft=~
      bIsGliderForward=~
      bIsGliderBack=~
      bHasDeployedAndIsStandingUpright=~
      PoseAdditive
     )))
  ]
 }
 AnimationTransitionGraph_1 {path=PetFullOverrideLayer.AnimGraphNode_StateMachine_1.SkyDiveMachine.AnimStateTransitionNode_4.AnimationTransitionGraph_1, nodes [Result()]}
 AnimationTransitionGraph_1 {path=PetFullOverrideLayer.AnimGraphNode_StateMachine_1.SkyDiveMachine.AnimStateTransitionNode_5.AnimationTransitionGraph_1, nodes [Result()]}
 AnimationTransitionGraph_1 {
  path=PetFullOverrideLayer.AnimGraphNode_StateMachine_1.SkyDiveMachine.AnimStateTransitionNode_6.AnimationTransitionGraph_1
  nodes [Result(bCanEnterTransition=Break Fort Anim Input Skydiving(FortAnimInput_Skydiving=Get SkydivingState()) -> (
     bIsSkydiveDiveMode=~
     bIsGliderRight=~
     bIsGliderCenter=~
     bIsGliderLeft=~
     bIsGliderForward=~
     bIsGliderBack=~
     bHasDeployedAndIsStandingUpright=~
    ))]
 }
 AnimationTransitionGraph_1 {
  path=PetFullOverrideLayer.AnimGraphNode_StateMachine_1.SkyDiveMachine.AnimStateTransitionNode_7.AnimationTransitionGraph_1
  nodes [
   n1 = Get SkydivingState()
   Result(bCanEnterTransition=NOT Boolean(A=Break Fort Anim Input Skydiving(FortAnimInput_Skydiving=@n1) -> (
      bIsSkydiveDiveMode=~
      bIsGliderRight=~
      bIsGliderCenter=~
      bIsGliderLeft=~
      bIsGliderForward=~
      bIsGliderBack=~
      bHasDeployedAndIsStandingUpright=~
     )))
  ]
 }
 AnimationTransitionGraph_0 {path=PetFullOverrideLayer.AnimGraphNode_StateMachine_1.SkyDiveMachine.AnimStateTransitionNode_8.AnimationTransitionGraph_0, nodes [Result()]}
 AnimationTransitionGraph_0 {
  path=PetFullOverrideLayer.AnimGraphNode_StateMachine_1.SkyDiveMachine.AnimStateTransitionNode_9.AnimationTransitionGraph_0
  nodes [Result(bCanEnterTransition=Get bTransition_ToSkydive())]
 }
 PetSkeletalControlLayer {nodes [
   n1 = Input Pose!() -> (Pose, LookAtOverrideInput:double, LookAtTargetLocationInput:Vector, LegIKAlphaOverride:double, HandIKAlphaOverride:double) {name=Pose}
   n2 = Look At - Bone: head!(ComponentPose=Local To Component(LocalPose=@n1), Alpha=@n1.LookAtOverrideInput, LookAtLocation:Vector=@n1.LookAtTargetLocationInput) -> (Pose) {lookAtClamp=50, alphaBoolBlend {blendOption=Linear}, boneToModify {boneName=head}, lookAt_Axis {axis=<-1 0>}, lookAtTarget=~}
   Output Pose(Result=Component To Local(ComponentPose=@n2)) {layerGroup=PetOverride}
  ]}
]
variables []
